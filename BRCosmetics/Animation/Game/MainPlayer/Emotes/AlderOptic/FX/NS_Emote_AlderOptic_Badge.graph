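# Niagara system for the badge effect of the AlderOptic emote, built from BadgePieces, Lightning and Wave emitters.
# BadgePieces uses staggered bursts of sprites and a mesh that rotate into place, then flash and flicker in brightness.
# Lightning spawns short-lived, randomly mirrored, camera-locked mesh bolts; Wave adds timed expanding waves, rays, glows and halos.
# Spawn groups tag visibility, three-way partitions select colors, partly from User.Color Array, and front-end modifiers adjust scale and intensity.
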NiagaraSystem /BRCosmetics/Animation/Game/MainPlayer/Emotes/AlderOptic/FX/NS_Emote_AlderOptic_Badge
dependencies {
 usedDataInterfaces [
  &NiagaraDataInterfaceCurve
  &NiagaraDataInterfaceColorCurve
  &NiagaraDataInterfaceArrayFloat
  &NiagaraDataInterfaceCamera
  &NiagaraDataInterfaceSpriteRendererInfo
  &NiagaraDataInterfaceVectorCurve
 ]
 usedDynamicInputs [&VectorFromCurve]
 usedModules [
  &self:SystemSpawnScript.SystemScriptSource.SystemScriptGraph.NiagaraNodeAssignment_2.SetVariables_#1
  &SystemState
  &self:SystemSpawnScript.SystemScriptSource.SystemScriptGraph.NiagaraNodeAssignment_1.SetVariables_#2
 ]
}
emitters {
 BadgePieces {
  emitterUpdateScript [
   &EmitterState(
    Enable Distance Culling=true
    Enable Visibility Culling=true
    Life Cycle Mode=::NewEnumerator1
    Loop Behavior=::NewEnumerator1
    Loop Duration Mode=::NewEnumerator1
    Module.bUseMaxDistance=false
    Module.bUseMinDistance=false
    Scalability Mode=::NewEnumerator1
    UseLoopDelay=false
    Visibility Cull Response:ENiagaraExecutionStateManagement=::SleepAndClearParticles
   )
   &SpawnBurst_Instantaneous(Use Spawn Probability=false, UseLoopCountLimit=false)
   &SpawnBurst_Instantaneous(Spawn Group=1, Spawn Time=0.05, Use Spawn Probability=false, UseLoopCountLimit=false) {module=SpawnBurst_Instantaneous001}
   &SpawnBurst_Instantaneous(Spawn Group=2, Spawn Time=0.1, Use Spawn Probability=false, UseLoopCountLimit=false) {module=SpawnBurst_Instantaneous002}
   &SpawnBurst_Instantaneous(Spawn Group=3, Use Spawn Probability=false, UseLoopCountLimit=false) {module=SpawnBurst_Instantaneous003}
   &SpawnBurst_Instantaneous(Spawn Count=3, Spawn Group=4, Use Spawn Probability=false, UseLoopCountLimit=false) {module=SpawnBurst_Instantaneous004}
  ]
  particleSpawnScript [
   /SetVariables(Particles.VisibilityTag=&SpawnGroup()) {isSetParameters=true}
   &PartitionParticles(Number Of Partitions=3, Particles Per Partition=1, Partition Mode:ENiagara_PartitionMode_v2=::NewEnumerator1)
   &InitializeParticle(
    Color=&MaskLinearColorBySpawnGroup(
     Masked Linear Color=&SelectLinearColorFromArray(
      Array Sampling Mode=::NewEnumerator1
      Direct Array Index=$Particles.PartitionParticles.Partition
      Color Selection Array=$User.Color Array
      Random Seed=?
     )
     Passthrough Linear Color=<rgba 1 1 1 1>
     Spawn Group Mask=4
    )
    Lifetime=&MaskFloat(Mask Source=0, Mask Parameter=?, Mask=1, Masked Float=5, Passthrough Float=1)
    Mesh Scale Mode=::NewEnumerator3
    Mesh Uniform Scale=0.7
    Sprite Rotation Angle=&Multiply_FloatByInt(Float=120, Integer=&SpawnGroup())
    Sprite Rotation Mode=::NewEnumerator0
    Sprite Size Mode=::NewEnumerator3
    Uniform Sprite Size=30
    UsePositionOffset=false
    Write Material Random=true
   )
   &CameraOffset(Camera Offset Amount=1)
   &InitialMeshOrientation(
    Mesh Orientation Mode=::NewEnumerator1
    Rotation=&MakeVector(X=&RandomRangeFloat(Minimum=0, Maximum=1), Y=0, Z=0)
    Use Rotation Vector=true
   ) {module=InitialMeshOrientation001}
  ]
  particleUpdateScript [
   &ParticleState(Kill Particles When Lifetime Has Elapsed=false, Let Infinitely Lived Particles Die When Emitter Deactivates=true)
   &SystemLocation(MaskSpawnGroup=false, Offset:NiagaraPosition=$System.SystemIdlePosition)
   &DynamicMaterialParameters(
    Index 0 Param 1=&RandomRangeFloat(Minimum=0, Maximum=6)
    Index 0 Param 2=2
    Index 0 Param 3=3
    Module.Index0Float0Write=true
    Module.Index0Float1Write=true
    Module.Index0Float2Write=true
    Module.Index0Float3Write=true
   ) {module=DynamicMaterialParameters002}
   &ScaleColor(
    Scale Alpha=&FloatFromCurve(
     FloatCurve=DataInterface() {propertyValues {Curve {defaultValue=1, keys [{interpMode=RCIM_Cubic, tangentMode=RCTM_User, time=0, value=0}, {tangentMode=RCTM_User, time=0.15, value=1}]}}}
     CurveIndex=$Particles.Age
    )
    Scale RGB=&Multiply_VectorByFloat(
     Vector=<1 1 1>
     Float=&FloatFromCurve(
      FloatCurve=DataInterface() {propertyValues {Curve {keys [
          {arriveTangent=0.6302, interpMode=RCIM_Cubic, leaveTangent=0.6302, tangentMode=RCTM_User, time=0.1, value=0.1}
          {arriveTangent=2.3565, interpMode=RCIM_Cubic, leaveTangent=2.3565, time=0.3509, value=0.3564}
          {arriveTangent=10.5738, interpMode=RCIM_Cubic, leaveTangent=10.5738, tangentMode=RCTM_User, time=0.397, value=0.8}
          {arriveTangent=0.6242, interpMode=RCIM_Cubic, leaveTangent=0.6242, tangentMode=RCTM_User, time=0.415, value=3}
          <at 0.4859 1>
         ]}}}
      CurveIndex=$Emitter.Age
     )
    )
    ScaleA=true
    ScaleRGB=true
   )
   &ScaleSpriteSize(
    Uniform Curve Index=$Particles.Age
    Uniform Curve Sprite Scale=DataInterface() {propertyValues {Curve {keys [
        {arriveTangent=-5.433, interpMode=RCIM_Cubic, leaveTangent=-5.433, tangentMode=RCTM_User, time=0, value=3}
        {interpMode=RCIM_Cubic, tangentMode=RCTM_User, time=0.2, value=1}
        {tangentMode=RCTM_User, time=0.5, value=1}
       ]}}}
   )
   /SetVariables(
    Particles.BloomColor:LinearColor=&Multiply_LinearColorByFloat(
     Linear Color=&ColorFromCurve(
      ColorCurve=DataInterface() {propertyValues {
        AlphaCurve {keys [
          {arriveTangent=0.375, interpMode=RCIM_Cubic, time=0.3, value=0}
          {arriveTangent=1, interpMode=RCIM_Cubic, leaveTangent=1, time=0.4, value=0.7}
          <at 0.6 0.3 RCIM_Cubic>
         ]}
        BlueCurve {keys [<at 0 1 RCIM_Cubic>,<at 1 1 RCIM_Cubic>]}
        GreenCurve {keys [<at 0 1 RCIM_Cubic>,<at 1 1 RCIM_Cubic>]}
        RedCurve {keys [<at 0 1 RCIM_Cubic>,<at 1 1 RCIM_Cubic>]}
       }}
      CurveIndex=$Emitter.Age
     )
     Float=1
    )
    Particles.BloomDynParam:Vector4f=<0.05 2 0 0>
    Particles.DarkGlowColor:LinearColor=&MultiplyLinearColor(
     Linear Color=$Particles.Color
     Scale Factor=&Vector4FromCurve(
      CurveIndex=$Emitter.Age
      Vector4Curve=DataInterface() {propertyValues {
        AlphaCurve {keys [
          {interpMode=RCIM_Cubic, tangentMode=RCTM_User, time=0.3, value=0}
          {interpMode=RCIM_Cubic, tangentMode=RCTM_User, time=0.4, value=0.6}
          <at 1 0.2>
         ]}
        BlueCurve {keys [<at 0 0.1>,<at 1 0.1>]}
        GreenCurve {keys [<at 0 0.1>,<at 1 0.1>]}
        RedCurve {keys [<at 0 0.1>,<at 1 0.1>]}
       }}
     )
    )
    Particles.DarkGlowSize:Vector2f=&Multiply_Vector2D_ByFloat(Float=1.9, Vector2D=$Particles.SpriteSize)
    Particles.GlowSize=&Multiply_Vector2D_ByFloat(Float=2.1, Vector2D=$Particles.SpriteSize)
   ) {isSetParameters=true}
   /SetVariables(
    Particles.BloomColor:LinearColor=&Multiply_LinearColorByFloat(
     Linear Color=&MultiplyLinearColor(
      Linear Color=$Particles.BloomColor
      Scale Factor=&Vector4FromCurve(
       CurveIndex=$Emitter.NormalizedLoopAge
       Vector4Curve=DataInterface() {propertyValues {
         AlphaCurve {keys [
           {interpMode=RCIM_Cubic, leaveTangentWeight=0.3333, tangentMode=RCTM_User, time=0, value=1}
           {arriveTangentWeight=0.05, interpMode=RCIM_Cubic, leaveTangentWeight=0.2833, tangentMode=RCTM_User, time=0.15, value=1}
           {
            arriveTangent=-0.2768
            arriveTangentWeight=0.0173
            interpMode=RCIM_Cubic
            leaveTangent=-0.2768
            leaveTangentWeight=0.0518
            tangentMode=RCTM_User
            time=0.202
            value=1.0413
           }
           {
            arriveTangent=-0.0897
            arriveTangentWeight=0.0144
            interpMode=RCIM_Cubic
            leaveTangent=-0.0897
            leaveTangentWeight=0.0374
            tangentMode=RCTM_User
            time=0.2453
            value=0.895
           }
           {
            arriveTangent=-0.5011
            arriveTangentWeight=0.0173
            interpMode=RCIM_Cubic
            leaveTangent=-0.5011
            leaveTangentWeight=0.0201
            tangentMode=RCTM_User
            time=0.2972
            value=1.0327
           }
           {arriveTangentWeight=0.0692, interpMode=RCIM_Cubic, leaveTangentWeight=0.0658, tangentMode=RCTM_User, time=0.3575, value=0.8488}
           {arriveTangentWeight=0.135, interpMode=RCIM_Cubic, leaveTangentWeight=0.0422, tangentMode=RCTM_User, time=0.555, value=1.2729}
           {
            arriveTangent=-0.0054
            arriveTangentWeight=0.0178
            interpMode=RCIM_Cubic
            leaveTangent=-0.0054
            leaveTangentWeight=0.0244
            tangentMode=RCTM_User
            time=0.6084
            value=1
           }
           {
            arriveTangent=-0.0054
            arriveTangentWeight=0.1772
            interpMode=RCIM_Cubic
            leaveTangent=-0.0054
            leaveTangentWeight=0.1061
            tangentMode=RCTM_User
            time=0.6816
            value=0.8429
           }
           {
            arriveTangent=-0.024
            arriveTangentWeight=0.0244
            interpMode=RCIM_Cubic
            leaveTangent=-0.024
            leaveTangentWeight=0.0817
            tangentMode=RCTM_User
            time=0.7549
            value=1.0948
           }
           {
            arriveTangent=0.0264
            arriveTangentWeight=0.0199
            interpMode=RCIM_Cubic
            leaveTangent=0.0264
            leaveTangentWeight=0.0618
            tangentMode=RCTM_User
            time=0.8145
            value=0.9021
           }
           {
            arriveTangent=0.0088
            arriveTangentWeight=0.0239
            interpMode=RCIM_Cubic
            leaveTangent=0.0088
            leaveTangentWeight=0.038
            tangentMode=RCTM_User
            time=0.8861
            value=1.0431
           }
           {arriveTangent=-0.3789, interpMode=RCIM_Cubic, leaveTangent=-0.3789, time=0.9301, value=0.9292}
           <at 1 1>
          ]}
         BlueCurve {keys [<at 0 1>,<at 1 1>]}
         GreenCurve {keys [<at 0 1>,<at 1 1>]}
         RedCurve {keys [<at 0 1>,<at 1 1>]}
        }}
      )
     )
     Float=&FN_FrontEndFloatModifier(Gameplay=1, Front End=0.7)
    )
    Particles.SpriteRotation:NiagaraFloat=&Add_Float(
     A=$Particles.Initial.SpriteRotation
     B=&FloatFromCurve(
      FloatCurve=DataInterface() {propertyValues {Curve {keys [
          {arriveTangent=205.4638, interpMode=RCIM_Cubic, leaveTangent=205.4638, tangentMode=RCTM_User, time=0, value=-60}
          {arriveTangent=-4.4282, interpMode=RCIM_Constant, tangentMode=RCTM_User, time=0.4, value=0}
          <at 1.0057 0 RCIM_Cubic>
         ]}}}
      CurveIndex=$Emitter.Age
     )
    )
   ) {isSetParameters=true}
   &FN_PerParticleLOD(Enable SpriteSize Distance Scaling=true, Sprite Size Scale Far=<2.5 2.5>)
   /SetVariables(Particles.RaysColor:LinearColor=&Multiply_LinearColorByFloat(
     Linear Color=&ScaleLinearColorByCurve(
      InitialValue=$Particles.Initial.Color
      ScaleCurve=DataInterface() {propertyValues {
        AlphaCurve {keys [{interpMode=RCIM_Cubic, tangentMode=RCTM_User, time=0, value=1}, <at 1 1>]}
        BlueCurve {keys [{arriveTangent=-56.6576, interpMode=RCIM_Cubic, leaveTangent=-56.6576, tangentMode=RCTM_User, time=0.7, value=7}, <at 1 3>]}
        GreenCurve {keys [{arriveTangent=-56.6576, interpMode=RCIM_Cubic, leaveTangent=-56.6576, tangentMode=RCTM_User, time=0.7, value=7}, <at 1 3>]}
        RedCurve {keys [{arriveTangent=-56.6576, interpMode=RCIM_Cubic, leaveTangent=-56.6576, tangentMode=RCTM_User, time=0.7, value=7}, <at 1 3>]}
       }}
      CurveIndex=$Particles.NormalizedAge
     )
     Float=&FloatFromCurve(
      FloatCurve=DataInterface() {propertyValues {Curve {keys [
          {interpMode=RCIM_Cubic, tangentMode=RCTM_User, time=0.3, value=0}
          {interpMode=RCIM_Cubic, tangentMode=RCTM_User, time=0.4, value=3}
          {interpMode=RCIM_Cubic, tangentMode=RCTM_User, time=1, value=0.7}
         ]}}}
      CurveIndex=$Emitter.Age
     )
    )) {isSetParameters=true}
  ]
  rendererClasses [&NiagaraSpriteRendererProperties, &NiagaraMeshRendererProperties]
  renderers [
   {rendererClass=&NiagaraSpriteRendererProperties}
   {rendererClass=&NiagaraSpriteRendererProperties, rendererIndex=1}
   {rendererClass=&NiagaraSpriteRendererProperties, rendererIndex=2}
   {rendererClass=&NiagaraSpriteRendererProperties, rendererIndex=3}
   {rendererClass=&NiagaraSpriteRendererProperties, rendererIndex=4}
   {rendererClass=&NiagaraSpriteRendererProperties, rendererIndex=5}
   {rendererClass=&NiagaraMeshRendererProperties, rendererIndex=6}
  ]
 }
 Lightning {
  emitterUpdateScript [
   &EmitterState(Life Cycle Mode=::NewEnumerator1, Loop Duration=2.3, UseLoopDelay=false)
   &SpawnRate(SpawnGroup=2, SpawnRate=5, Use Spawn Probability=false)
   &SpawnRate(
    SpawnGroup=3
    SpawnRate=&FloatFromCurve(
     FloatCurve=DataInterface() {propertyValues {Curve {keys [
         <at 0 0 RCIM_Cubic>
         <at 0.1196 0.6932 RCIM_Cubic>
         {arriveTangent=1.1125, interpMode=RCIM_Cubic, leaveTangent=1.1125, tangentMode=RCTM_User, time=0.3135, value=0}
         {arriveTangent=1.2369, interpMode=RCIM_Cubic, leaveTangent=1.2369, time=0.4343, value=3.4471}
         {arriveTangent=0.1357, interpMode=RCIM_Cubic, leaveTangent=0.1357, tangentMode=RCTM_User, time=0.6922, value=0.4684}
         {arriveTangent=-1.5218, interpMode=RCIM_Cubic, leaveTangent=-1.5218, time=0.8588, value=1}
         <at 1 0 RCIM_Cubic>
        ]}}}
     CurveIndex=$Emitter.NormalizedLoopAge
    )
    Use Spawn Probability=false
   ) {module=SpawnRate001}
   &SpawnBurst_Instantaneous(Spawn Count=9, Spawn Group=2, Spawn Time=0.2, Use Spawn Probability=false, UseLoopCountLimit=true) {module=SpawnBurst_Instantaneous004}
   &SpawnBurst_Instantaneous(Spawn Count=3, Spawn Group=3, Spawn Time=0.2, Use Spawn Probability=false, UseLoopCountLimit=true) {module=SpawnBurst_Instantaneous002}
  ]
  particleSpawnScript [
   /SetVariables(
    Particles.LightningIndex:NiagaraInt32=&RandomRangeInt(Minimum=0, Maximum=3)
    Particles.LightningLength:NiagaraFloat=&Multiply_Float(
     A=&RandomRangeFloat(Minimum=0.7, Maximum=1)
     B=&FN_FrontEndFloatModifier(Gameplay=1, Front End=0.45)
    )
    Particles.MirrorLightning:NiagaraBool=&RandomBool()
   ) {isSetParameters=true}
   &PartitionParticles(Number Of Partitions=3, Partition Mode:ENiagara_PartitionMode_v2=::NewEnumerator5)
   &InitializeParticle(
    AdjustAlpha=false
    AdjustHue=false
    AdjustSaturation=false
    AdjustValue=true
    Color=&SelectLinearColorFromArray(
     Array Sampling Mode=::NewEnumerator1
     Direct Array Index=$Particles.PartitionParticles.Partition
     Color Selection Array=$User.Color Array
     Random Seed=?
    )
    Color Mode=::NewEnumerator3
    Lifetime=0.3
    UsePositionOffset=false
    Value Range:Vector2f=<0.8 3>
    Write Material Random=true
   )
   &ShapeLocation(MaskSpawnGroup=false, Override Seed=false, Sphere Radius=1, UseOwnerScale=true)
   &FN_PerParticleLOD(Enable SpriteSize Distance Scaling=true, Sprite Size Scale Far=<2.5 2.5>)
   /SetVariables(Particles.VisibilityTag=&SpawnGroup()) {isSetParameters=true}
   &InitialMeshOrientation(
    Mesh Orientation Mode=::NewEnumerator2
    Orientation Coordinate Space=::Local
    Orientation Vector=&FN_FrontEndVector3Modifier(Gameplay=<-1 0 0>, Front End=&RandomVector(Vector Scale=1))
    Rotation=&MakeVector(
     X=0
     Y=&Add_Float(
      A=&SelectFloatFromArray(
       Direct Array Index=$Particles.PartitionParticles.Partition
       Float Selection Array=DataInterface() {propertyValues {FloatData [0,0.3333,0.6666]}}
      )
      B=&Add_Float(A=-0.25, B=&RandomRangeFloat(Minimum=-0.05, Maximum=0.05))
     )
     Z=0
    )
    Use Rotation Vector=true
   )
   &CameraOffset(Camera Offset Amount=5)
  ]
  particleUpdateScript [
   &ParticleState()
   /SetVariables(
    Particles.DynamicParametersAlt:Vector4f=&MakeVector4(
     X=&Multiply_Float(A=$Particles.LightningLength, B=&RandomRangeFloat(Minimum=18, Maximum=50))
     Y=&UniformAOrBFloat(A=1, B=2, Evaluation Type=::NewEnumerator0, Recalculate Random Each Loop=true, Override Seed=false)
     Z=&FloatFromCurve(FloatCurve=DataInterface() {propertyValues {Curve {
         defaultValue=1
         keys [
          {arriveTangent=-2.478, interpMode=RCIM_Cubic, leaveTangent=-2.478, tangentMode=RCTM_User, time=0, value=1}
          {interpMode=RCIM_Cubic, tangentMode=RCTM_User, time=1, value=0}
         ]
        }}})
     W=0
    )
    Particles.DynamicParametersAlt2:Vector4f=&MakeVector4(X=1, Y=1, Z=1, W=0)
   ) {isSetParameters=true}
   &FN_Mesh_Attributes(
    Create Scale Attribute=true
    Scale=&Multiply_Vector(
     A=&Multiply_VectorByFloat(
      Vector=&MakeCustomVectorFromBool(Bool=$Particles.MirrorLightning, True Vector=<1 -1 -1>, False Vector=<1 1 1>)
      Float=$Particles.LightningLength
     )
     B=&Multiply_VectorByFloat(
      Vector=&MakeVector(
       X=&FloatFromCurve(FloatCurve=DataInterface() {propertyValues {Curve {
           defaultValue=1
           keys [
            {arriveTangent=0.8532, interpMode=RCIM_Cubic, leaveTangent=0.8532, tangentMode=RCTM_User, time=0, value=0.8}
            {interpMode=RCIM_Cubic, tangentMode=RCTM_User, time=1, value=1.4}
           ]
          }}})
       Y=1
       Z=1
      )
      Float=$System.FrontendScale
     )
    )
   )
   /SetVariables(
    Particles.DynamicParametersAlt1:Vector4f=&MakeVector4(X=0, Y=0, Z=0, W=13)
    Particles.DynamicParametersAlt3:Vector4f=&MakeVector4(
     X=&FloatFromCurve(FloatCurve=DataInterface() {propertyValues {Curve {keys [<at 0 0.5>,<at 1 1>]}}})
     Y=2
     Z=&FloatFromCurve(FloatCurve=DataInterface() {propertyValues {Curve {keys [{interpMode=RCIM_Cubic, tangentMode=RCTM_Break, time=0.5, value=-0.5}, <at 1 0.1>]}}})
     W=0
    )
    Particles.DynamicParametersAlt4:Vector4f=&MakeVector4(
     X=&Multiply_Float(
      A=&FloatFromCurve(
       FloatCurve=DataInterface() {propertyValues {Curve {keys [{interpMode=RCIM_Cubic, leaveTangent=8.1085, tangentMode=RCTM_Break, time=0.8568, value=1}, {tangentMode=RCTM_User, time=1, value=0}]}}}
       Scale Curve=0.8
      )
      B=&Divide_Float(
       A=&MakeFloatFromVector(Vector=$Particles.FN_Mesh_Attributes001.Scale, Channel=::NewEnumerator0)
       B=&RandomRangeFloat(Minimum=10, Maximum=15)
      )
     )
     Y=&FloatFromCurve(
      FloatCurve=DataInterface() {propertyValues {Curve {keys [
          {interpMode=RCIM_Cubic, tangentMode=RCTM_Break, time=0.7, value=0}
          {arriveTangent=-0.0859, interpMode=RCIM_Cubic, tangentMode=RCTM_Break, time=1, value=-1}
         ]}}}
      Scale Curve=50
     )
     Z=&Waveform(
      Global Input Position=&RandomRangeFloat(Minimum=0, Maximum=15)
      Global Phase=$Particles.Age
      "Waveform [1]"=::NewEnumerator12
      "[1] Phase Scale"=2
      "[1] Frequency"=1
      "[1] Amplitude Min/Max"=<0 1>
     )
     W=25
    )
    Particles.DynamicParametersAlt5:Vector4f=&MakeVector4(
     X=0
     Y=0
     Z=&FloatFromCurve(FloatCurve=DataInterface() {propertyValues {Curve {keys [
          {interpMode=RCIM_Cubic, leaveTangent=1.1034, tangentMode=RCTM_Break, time=0, value=-1}
          {arriveTangent=2.2399, tangentMode=RCTM_User, time=0.5, value=0.125}
         ]}}})
     W=&FloatFromCurve(FloatCurve=DataInterface() {propertyValues {Curve {keys [
          {interpMode=RCIM_Cubic, leaveTangent=1.8089, tangentMode=RCTM_Break, time=0.3, value=-0.5}
          {arriveTangent=2.0757, interpMode=RCIM_Cubic, tangentMode=RCTM_Break, time=0.9, value=1}
         ]}}})
    )
   ) {isSetParameters=true}
   &FN_Mesh_Attributes(
    Color=&MultiplyLinearColor(
     Linear Color=&RandomRangeLinearColor(Minimum=<rgba 5 1 5 1>, Maximum=<rgba 5 1 3.6667 1>)
     Scale Factor=&MakeVector4FromVectorAndFloat(
      "Vector (RGB)"=&VectorFromFloat(Value=&RandomRangeFloat(Minimum=0, Maximum=2, Evaluation Type=::NewEnumerator1))
      "Float (Alpha)"=&FloatFromCurve(FloatCurve=DataInterface() {propertyValues {Curve {keys [<at 0.7 1>,<at 1 0>]}}})
     )
    )
    Create Color Attribute=true
    Create Orientation Attribute=true
    Create Scale Attribute=true
    Orientation:Quat4f=&MakeQuaternion(
     Quaternion From=1
     Axis=&RandomRangeVector(Minimum=<-0.3 -0.3 -1>, Maximum=<0.3 0.3 1>)
     Angle=&RandomRangeFloat(Minimum=0, Maximum=360)
     Yaw=&RandomRangeFloat(Minimum=-0.3, Maximum=0.3)
     Pitch=&RandomRangeFloat(Minimum=-0.3, Maximum=0.3)
     Roll=&RandomRangeFloat(Minimum=-1, Maximum=1)
     From=?
     To=?
     Angle Type=::NewEnumerator1
     Coordinate Space=::Local
     Normalize Vectors=false
    )
    Scale=&Multiply_VectorByFloat(
     Vector=&VectorFromFloat(Value=&Multiply_Float(A=&RandomRangeFloat(Minimum=0.225, Maximum=0.2475), B=$System.FrontendScale))
     Float=&FloatFromCurve(FloatCurve=DataInterface() {propertyValues {Curve {keys [
          {arriveTangent=1.9115, interpMode=RCIM_Cubic, leaveTangent=1.9115, tangentMode=RCTM_User, time=0, value=0}
          {arriveTangent=0.0224, interpMode=RCIM_Cubic, leaveTangent=0.0224, tangentMode=RCTM_User, time=1, value=1}
         ]}}})
    )
    Write to Module Namespace=true
   ) {module=FN_Mesh_Attributes001}
   &ScaleColor(
    Scale Alpha=&FloatFromCurve(FloatCurve=DataInterface() {propertyValues {Curve {keys [<at 0 0 RCIM_Cubic>, {interpMode=RCIM_Cubic, tangentMode=RCTM_User, time=0.7, value=0.2}, <at 1 0>]}}})
    Scale RGB=&Multiply_VectorByFloat(
     Vector=<1 1 1>
     Float=&FloatFromCurve(FloatCurve=DataInterface() {propertyValues {Curve {keys [
          {interpMode=RCIM_Cubic, tangentMode=RCTM_User, time=0, value=0.1}
          {interpMode=RCIM_Cubic, tangentMode=RCTM_User, time=0.5, value=5}
          <at 1 1>
         ]}}})
    )
    ScaleA=true
    ScaleRGB=true
   )
   &self:GimbalLock(Gimbal Lock Direction:NiagaraPosition=&ReturnCameraVector(Camera Vectors=::NewEnumerator4, Coordinate Space=::Local, Origin Position=$Engine.Emitter.SimulationPosition)) {module=GimbalLock}
  ]
  rendererClasses [&NiagaraMeshRendererProperties]
  renderers [
   {rendererClass=&NiagaraMeshRendererProperties, rendererIndex=0}
   {rendererClass=&NiagaraMeshRendererProperties, rendererIndex=1}
  ]
 }
 Wave {
  emitterUpdateScript [
   &EmitterState(Life Cycle Mode=::NewEnumerator1, Loop Behavior=::NewEnumerator1, Loop Duration Mode=::NewEnumerator1, UseLoopDelay=false)
   &SpawnBurst_Instantaneous(Use Spawn Probability=false, UseLoopCountLimit=false)
   &SpawnBurst_Instantaneous(Spawn Time=0.03, Use Spawn Probability=false, UseLoopCountLimit=false) {module=SpawnBurst_Instantaneous001}
   &SpawnBurst_Instantaneous(Spawn Time=0.06, Use Spawn Probability=false, UseLoopCountLimit=false) {module=SpawnBurst_Instantaneous002}
   &SpawnBurst_Instantaneous(Spawn Group=1, Spawn Time=0.3, Use Spawn Probability=false, UseLoopCountLimit=false) {module=SpawnBurst_Instantaneous003}
   &SpawnBurst_Instantaneous(Spawn Group=1, Spawn Time=0.4, Use Spawn Probability=false, UseLoopCountLimit=false) {module=SpawnBurst_Instantaneous004}
   &SpawnBurst_Instantaneous(Spawn Group=1, Spawn Time=0.5, Use Spawn Probability=false, UseLoopCountLimit=false) {module=SpawnBurst_Instantaneous005}
   &SpawnBurst_Instantaneous(Spawn Group=3, Spawn Time=0.3, Use Spawn Probability=false, UseLoopCountLimit=false) {module=SpawnBurst_Instantaneous007}
   &SpawnRate(SpawnGroup=3, SpawnRate=1.8, Use Spawn Probability=false)
   &SpawnBurst_Instantaneous(Spawn Group=4, Spawn Time=0.3, Use Spawn Probability=false, UseLoopCountLimit=false) {module=SpawnBurst_Instantaneous008}
   &SpawnBurst_Instantaneous(Spawn Count=9, Spawn Group=5, Use Spawn Probability=false) {module=SpawnBurst_Instantaneous006}
  ]
  particleSpawnScript [
   &PartitionParticles(
    Normalized Index In Partition Includes One=true
    Number Of Partitions=3
    Particles Per Partition=1
    Partition Mode:ENiagara_PartitionMode_v2=::NewEnumerator1
   )
   /SetVariables(Particles.VisibilityTag=&SpawnGroup()) {isSetParameters=true}
   &InitializeParticle(
    Color=&SelectLinearColorFromArray(
     Array Sampling Mode=::NewEnumerator1
     Direct Array Index=$Particles.PartitionParticles.Partition
     Color Selection Array=DataInterface() {propertyValues {ColorData [<rgba 3 0.03 1.8585 1>, <rgba 0 2 0.1929 1>, <rgba 0 1.6 3 1>]}}
     Random Seed=?
    )
    Lifetime=&SelectFloatFromArray(
     Direct Array Index=&SpawnGroup()
     Float Selection Array=DataInterface() {propertyValues {FloatData [0.5,0.7,1.1,1.4,0.5,0.7]}}
    )
    Mesh Scale Mode=::NewEnumerator3
    Mesh Uniform Scale=&Lerp_Float(
     A=&SelectFloatFromArray(Direct Array Index=$Particles.PartitionParticles.Partition, Float Selection Array=DataInterface() {propertyValues {FloatData [0.9,0.8,0.9]}})
     B=&SelectFloatFromArray(
      Direct Array Index=$Particles.PartitionParticles.Partition
      Float Selection Array=DataInterface() {propertyValues {FloatData [0.6,0.55,0.5]}}
     )
     Alpha=&MakeFloatFromInt(Int=&SpawnGroup())
    )
    Sprite Rotation Mode=::NewEnumerator2
    Sprite Size Mode=::NewEnumerator3
    Uniform Sprite Size=150
    UsePositionOffset=false
    Write Material Random=true
   )
   &InitialMeshOrientation(
    Orientation Vector=<0 0 1>
    Rotation=&RandomRangeVector(Minimum=<0 0 0>, Maximum=<1 0 0>)
    Use Orientation Vector=false
    Use Rotation Vector=true
   ) {module=InitialMeshOrientation001}
   /SetVariables(
    Particles.RaysColor:LinearColor=&SelectLinearColorFromArray(
     Array Sampling Mode=::NewEnumerator1
     Direct Array Index=$Particles.PartitionParticles.Partition
     Color Selection Array=$User.Color Array
     Random Seed=?
    )
    Particles.RaysRotation:NiagaraFloat=&Subtract_Float(
     A=&Multiply_Float(
      A=120
      B=&MakeFloatFromInt(Int=&Add_Integer(A=$Particles.PartitionParticles.Partition, B=1))
     )
     B=&Add_Float(A=120, B=&RandomRangeFloat(Minimum=-50, Maximum=25))
    )
    Particles.RaysScale:Vector2f=&Multiply_Vector2D_ByFloat(Float=0.2, Vector2D=&RandomRangeVector2D(Minimum=<150 80>, Maximum=<200 100>))
   ) {isSetParameters=true}
   &SubUVAnimation(
    Evaluation Type=::NewEnumerator0
    "Sprite Renderer "=DataInterface() {propertyValues {SpriteRenderer=&self:Wave_0.NiagaraSpriteRendererProperties_3}}
    SubUV Animation Mode:ENiagara_SubUVLookupModeV2=::NewEnumerator2
    UseEndFrame=false
    UseStartFrame=false
   ) {enabled=false}
   /SetVariables() {isSetParameters=true}
  ]
  particleUpdateScript [
   &ParticleState()
   &ScaleSpriteSize(Scale Factor=&Multiply_Vector2D_ByFloat(
     Float=&FloatFromCurve(
      FloatCurve=DataInterface() {propertyValues {Curve {keys [
          {arriveTangent=1.0289, interpMode=RCIM_Cubic, leaveTangent=1.0289, tangentMode=RCTM_User, time=0, value=0.3}
          {interpMode=RCIM_Cubic, tangentMode=RCTM_User, time=1, value=1}
         ]}}}
      Scale Curve=1.25
     )
     Vector2D=<1 1>
    ))
   &ScaleMeshSize(Scale Factor=&Multiply_VectorByFloat(
     Vector=<1 1 1>
     Float=&FloatFromCurve(FloatCurve=DataInterface() {propertyValues {Curve {keys [{arriveTangent=-3.5595, interpMode=RCIM_Cubic, leaveTangent=-3.5595, tangentMode=RCTM_User, time=0, value=2}, <at 1 1>]}}})
    ))
   &ScaleColor(
    Scale Alpha=&FloatFromCurve(
     FloatCurve=DataInterface() {propertyValues {Curve {keys [
         <at 0 0 RCIM_Cubic>
         {arriveTangentWeight=0.0373, interpMode=RCIM_Cubic, leaveTangentWeight=0.3896, tangentMode=RCTM_User, time=0.5, value=0.5}
         <at 1 0>
        ]}}}
     Scale Curve=&FN_FrontEndFloatModifier(Gameplay=1, Front End=0.7)
    )
    Scale RGB=&Multiply_VectorByFloat(
     Vector=<1 1 1>
     Float=&FloatFromCurve(FloatCurve=DataInterface() {propertyValues {Curve {keys [{arriveTangent=-40.5001, interpMode=RCIM_Cubic, leaveTangent=-40.5001, tangentMode=RCTM_User, time=0, value=15}, <at 1 3>]}}})
    )
    ScaleA=true
    ScaleRGB=true
   )
   &SystemLocation(MaskSpawnGroup=false, Offset:NiagaraPosition=$System.SystemIdlePosition)
   &DynamicMaterialParameters(
    Index 0 Param 1=&FloatFromCurve(FloatCurve=DataInterface() {propertyValues {Curve {keys [{arriveTangent=-2.1015, interpMode=RCIM_Cubic, leaveTangent=-2.1015, tangentMode=RCTM_User, time=0, value=0.5}, <at 1 -0.2917>]}}})
    Index 0 Param 2=&FloatFromCurve(FloatCurve=DataInterface() {propertyValues {Curve {keys [{arriveTangent=-0.3464, interpMode=RCIM_Cubic, leaveTangent=-0.3464, tangentMode=RCTM_User, time=0, value=1}, <at 1 0>]}}})
    Index 0 Param 3=&FloatFromCurve(FloatCurve=DataInterface() {propertyValues {Curve {
        defaultValue=1
        keys [
         {arriveTangent=-3.9897, interpMode=RCIM_Cubic, leaveTangent=-3.9897, tangentMode=RCTM_User, time=0, value=2}
         {tangentMode=RCTM_User, time=1, value=-0.5}
        ]
       }}})
    Index 0 Param 4=1
    Index 3 Param 2=&RandomRangeFloat(Minimum=0, Maximum=2)
    Module.Index0Float0Write=true
    Module.Index0Float1Write=true
    Module.Index0Float2Write=true
    Module.Index0Float3Write=false
    Module.Index3Float0Write=false
    Module.Index3Float1Write=true
    Module.Index3Float2Write=false
    Module.Index3Float3Write=false
    Write Parameter Index 3=::NewEnumerator1
   ) {module=DynamicMaterialParameters001}
   /SetVariables(
    Particles.WaveColor:LinearColor=&MultiplyLinearColor(
     Linear Color=$Particles.Initial.Color
     Scale Factor=/Niagara/DynamicInputs/Multiply/Multiply_Vector4_ByFloat(
      Vector4:Vector4f=&Vector4FromCurve(Vector4Curve=DataInterface() {propertyValues {
         AlphaCurve {keys [{interpMode=RCIM_Cubic, tangentMode=RCTM_User, time=0, value=0.3}, <at 1 0>]}
         BlueCurve {keys [<at 0 1>,<at 1 1>]}
         GreenCurve {keys [<at 0 1>,<at 1 1>]}
         RedCurve {keys [<at 0 1>,<at 1 1>]}
        }})
      Float:NiagaraFloat=&FN_FrontEndFloatModifier(Gameplay=1, Front End=0.5)
     )
    )
    Particles.WaveDynamicMaterialParameter:Vector4f=&MakeVector4(
     X=&FloatFromCurve(FloatCurve=DataInterface() {propertyValues {Curve {keys [{arriveTangent=-2.1015, interpMode=RCIM_Cubic, leaveTangent=-2.1015, tangentMode=RCTM_User, time=0, value=0.5}, <at 1 -0.15>]}}})
     Y=&FloatFromCurve(FloatCurve=DataInterface() {propertyValues {Curve {keys [{arriveTangent=-0.3016, interpMode=RCIM_Cubic, leaveTangent=-0.3016, tangentMode=RCTM_User, time=0, value=1}, <at 1 0.9>]}}})
     Z=&FloatFromCurve(FloatCurve=DataInterface() {propertyValues {Curve {keys [
          {arriveTangent=-0.0878, interpMode=RCIM_Cubic, leaveTangent=-0.0878, tangentMode=RCTM_User, time=0, value=0}
          {arriveTangent=1.2291, interpMode=RCIM_Cubic, leaveTangent=1.2291, tangentMode=RCTM_User, time=0.1859, value=0.3101}
          <at 1 0.8>
         ]}}})
     W=0
    )
    Particles.WaveScale:Vector3f=&Multiply_VectorByFloat(
     Vector=$Particles.Scale
     Float=&FloatFromCurve(FloatCurve=DataInterface() {propertyValues {Curve {
         defaultValue=1
         keys [
          {arriveTangent=2.0193, interpMode=RCIM_Cubic, leaveTangent=2.0193, tangentMode=RCTM_User, time=0, value=0.5}
          {arriveTangent=0.6897, interpMode=RCIM_Cubic, leaveTangent=0.6897, tangentMode=RCTM_User, time=0.5, value=1.3}
          <at 1 1.5>
         ]
        }}})
    )
   ) {isSetParameters=true}
   /SetVariables(Particles.RaysColor:LinearColor=&ScaleLinearColorByCurve(
     InitialValue=$Particles.Initial.RaysColor
     ScaleCurve=DataInterface() {propertyValues {
       AlphaCurve {keys [
         {interpMode=RCIM_Cubic, tangentMode=RCTM_User, time=0, value=0}
         {interpMode=RCIM_Cubic, tangentMode=RCTM_User, time=0.3088, value=1}
         <at 1 0>
        ]}
       BlueCurve {keys [<at 0 1>]}
       GreenCurve {keys [<at 0 1>]}
       RedCurve {keys [<at 0 1>]}
      }}
     CurveIndex=$Particles.NormalizedAge
    )) {isSetParameters=true}
   /SetVariables(Particles.Corona_Size:Vector2f=&Multiply_Vector2D_ByFloat(Float=0.5, Vector2D=$Particles.SpriteSize)) {isSetParameters=true}
   /SetVariables(
    Particles.GlowColor=&Multiply_LinearColorByFloat(
     Linear Color=&ColorFromCurve(ColorCurve=DataInterface() {propertyValues {
        AlphaCurve {keys [<at 0 1 RCIM_Cubic>,<at 0.1869 0.5>,<at 0.9 0>]}
        BlueCurve {keys [<at 0 0>,<at 0.6 1>]}
        GreenCurve {keys [<at 0 1.5026>,<at 0.6 0>]}
        RedCurve {keys [<at 0 3>,<at 0.6 0.3051>]}
       }})
     Float=&FN_FrontEndFloatModifier(Gameplay=1, Front End=0.8)
    )
    Particles.GlowSize=&Vector2DFromFloat(Value=&FloatFromCurve(
      FloatCurve=DataInterface() {propertyValues {Curve {keys [
          {arriveTangent=11.8116, interpMode=RCIM_Cubic, leaveTangent=11.8116, tangentMode=RCTM_User, time=0, value=0}
          {arriveTangent=-0.0883, interpMode=RCIM_Cubic, leaveTangent=-0.0883, tangentMode=RCTM_User, time=0.1, value=1.2}
          {arriveTangent=-0.2122, interpMode=RCIM_Cubic, leaveTangent=-0.2122, tangentMode=RCTM_User, time=0.25, value=1}
          {interpMode=RCIM_Cubic, leaveTangent=-2.2807, time=1, value=1}
         ]}}}
      Scale Curve=130
     ))
    Particles.HaloColor:LinearColor=&Multiply_LinearColorByFloat(
     Linear Color=&ColorFromCurve(ColorCurve=DataInterface() {propertyValues {
        AlphaCurve {keys [<at 0 0 RCIM_Cubic>,<at 0.2571 0.5>,<at 1 0 RCIM_Cubic>]}
        BlueCurve {keys [<at 0 2.6146>,<at 0.3862 2>,<at 0.8983 0.3018>]}
        GreenCurve {keys [<at 0 0>,<at 0.3862 0.7428>,<at 0.8983 1.3224>]}
        RedCurve {keys [<at 0 3>,<at 0.3862 0.0149>,<at 0.8983 0.107>]}
       }})
     Float=&FN_FrontEndFloatModifier(Gameplay=1, Front End=0.5)
    )
   ) {isSetParameters=true}
  ]
  rendererClasses [&NiagaraMeshRendererProperties, &NiagaraSpriteRendererProperties]
  renderers [
   {rendererClass=&NiagaraMeshRendererProperties, rendererIndex=0}
   {rendererClass=&NiagaraMeshRendererProperties, rendererIndex=1}
   {rendererClass=&NiagaraSpriteRendererProperties, rendererIndex=2}
   {rendererClass=&NiagaraSpriteRendererProperties, rendererIndex=3}
   {rendererClass=&NiagaraSpriteRendererProperties, rendererIndex=4}
   {rendererClass=&NiagaraSpriteRendererProperties, rendererIndex=5}
  ]
 }
}
summary {
 systemName=NS_Emote_AlderOptic_Badge
 emitters [
  {emitterName=BadgePieces, rendererClasses [&NiagaraSpriteRendererProperties, &NiagaraMeshRendererProperties]}
  {emitterName=Lightning, rendererClasses [&NiagaraMeshRendererProperties]}
  {emitterName=Wave, rendererClasses [&NiagaraMeshRendererProperties, &NiagaraSpriteRendererProperties]}
 ]
 userVariables [
  User.bIsFrontEnd {type=NiagaraBool}
  User.bIsFrontEndPreview {type=NiagaraBool}
  User.Color Array {
   type=NiagaraDataInterfaceArrayColor
   description=""
   defaultValue {
    struct=NiagaraExt_VariableValue_DataInterface
    value {
     dataInterface=&self:NiagaraDataInterfaceArrayColor_0
     dataInterfaceClass=&NiagaraDataInterfaceArrayColor
    }
   }
  }
 ]
}
