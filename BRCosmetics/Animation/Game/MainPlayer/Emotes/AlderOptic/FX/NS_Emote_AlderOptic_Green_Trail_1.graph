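# Niagara system for the green swing trail in the AlderOptic emote, drawing two animation trails from bone or socket pairs on a skeletal mesh.
# MainTrail_Packed feeds five ribbon renderers from one particle set via alternate width, color and material-parameter attributes, with lightning-style jitter along ribbon facing.
# TipTrail adds a narrower two-renderer ribbon.
# Per-trail system swing fades drive alpha in both emitters, with time-of-day scaling on the main trail.
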NiagaraSystem /BRCosmetics/Animation/Game/MainPlayer/Emotes/AlderOptic/FX/NS_Emote_AlderOptic_Green_Trail_1
dependencies {
 usedDataInterfaces [&NiagaraDataInterfaceCurve, &NiagaraDataInterfaceColorCurve, &NiagaraDataInterfaceSocketReader]
 usedDynamicInputs [&FNTimeOfDayScalar]
 usedModules [
  &self:SystemSpawnScript.SystemScriptSource.SystemScriptGraph.NiagaraNodeAssignment_0.SetVariables_#1
  &self:SystemSpawnScript.SystemScriptSource.SystemScriptGraph.NiagaraNodeAssignment_1.SetVariables_#2
  &self:SystemSpawnScript.SystemScriptSource.SystemScriptGraph.NiagaraNodeAssignment_2.SetVariables_#3
  &SystemState
  &self:SystemSpawnScript.SystemScriptSource.SystemScriptGraph.NiagaraNodeAssignment_4.SetVariables_#4
 ]
}
emitters {
 MainTrail_Packed {
  emitterSpawnScript [
   &FN_RibbonIDOffsetUtility()
   /SetVariables(Emitter.RandomUOffset:NiagaraFloat=&RandomRangeFloat(Minimum=-1, Maximum=1)) {isSetParameters=true}
  ]
  emitterUpdateScript [
   &EmitterState()
   &FN_SpawnAnimTrail(AnimTrailWidthMode=::FromSecond, Number of Anim Trails=2)
  ]
  particleSpawnScript [
   &FN_InitializeAnimTrailBoneIndecies()
   &FN_InitializeAnimTrailBoneIndecies(AnimTrail Selection=1, First Bone or Socket Index=2, Second Bone or Socket Index=3) {module=FN_InitializeAnimTrailBoneIndecies001}
   &FN_InitializeAnimTrail(AnimTrailWidth=0.75, Lifetime=0.22)
  ]
  particleUpdateScript [
   &FN_UpdateAnimTrail(Lock Particles=true) {module=FN_UpdateAnimTrail001}
   &JitterPosition(
    Jitter Amount:NiagaraFloat=&FloatFromCurve(FloatCurve=DataInterface() {propertyValues {Curve {keys [<at 0 0>,<at 0.05 1>,<at 0.8 1>,<at 1 0>]}}}, Scale Curve=5)
    Jitter Offset:Vector3f=&Multiply_Vector(A=$System.RandomJitterVector, B=$Particles.RibbonFacing)
   )
   &ScaleRibbonWidth(Ribbon Width Scale=&FloatFromCurve(FloatCurve=DataInterface() {propertyValues {Curve {keys [<at 0.4 1>,<at 0.75 0>]}}}))
   /SetVariables(
    Particles.DynamicParametersAlt2:Vector4f=&MakeVector4(X=0, Y=0, Z=0, W=0)
    Particles.DynamicParametersAlt3:Vector4f=&MakeVector4(X=1, Y=0, Z=0, W=0)
    Particles.DynamicParametersAlt7:Vector4f=&MakeVector4(X=20, Y=0, Z=0, W=0)
   ) {isSetParameters=true}
   /SetVariables(Particles.RibbonWidthAlt3:NiagaraFloat=&Multiply_Float(
     A=$Particles.Initial.RibbonWidth
     B=&FloatFromCurve(FloatCurve=DataInterface() {propertyValues {Curve {keys [<at 0.7 2>,<at 1 0>]}}}, Scale Curve=1.5)
    )) {isSetParameters=true}
   &FN_Sprite_Attributes(
    Color=&MakeLinearColorFromVectorAndFloat(
     "Vector (RGB)"=&MakeVectorFromLinearColorRGB(Color=<rgba 0.023 0.2 0.026 1>)
     "Float (Alpha)"=&Multiply_Float(
      A=&Lerp_Float(
       A=$System.SwingGeneralFade1
       B=$System.SwingGeneralFade2
       Alpha=&MakeFloatFromInt(Int=&SpawnGroup())
      )
      B=&FloatFromCurve(
       FloatCurve=DataInterface() {propertyValues {Curve {keys [<at 0 0>,<at 0.07 1>]}}}
       CurveIndex=$Emitter.Age
       Scale Curve=&Multiply_Float(A=0.2, B=$System.ToDScalar)
      )
     )
    )
    Create Color Attribute=true
    Write to Module Namespace=true
   ) {module=FN_Sprite_Attributes001}
   /SetVariables(
    Particles.DynamicParametersAlt4:Vector4f=&MakeVector4(X=1, Y=0, Z=0, W=0)
    Particles.DynamicParametersAlt5:Vector4f=&MakeVector4(X=3, Y=3, Z=0, W=0)
    Particles.DynamicParametersAlt6:Vector4f=&MakeVector4(X=5, Y=18, Z=0, W=0)
   ) {isSetParameters=true}
   /SetVariables(Particles.RibbonWidthAlt:NiagaraFloat=&Multiply_Float(
     A=$Particles.RibbonWidth
     B=&FloatFromCurve(FloatCurve=DataInterface() {propertyValues {Curve {keys [{arriveTangent=-25.047, time=0, value=1.65}, <at 0.3 0>]}}})
    )) {isSetParameters=true}
   &FN_Sprite_Attributes(
    Color=&MakeLinearColorFromVectorAndFloat(
     "Vector (RGB)"=&MakeVectorFromLinearColorRGB(Color=&ColorFromCurve(ColorCurve=DataInterface() {propertyValues {
         AlphaCurve {keys []}
         BlueCurve {keys [<at 0 2.375 RCIM_Cubic>,<at 1 2.5 RCIM_Cubic>]}
         GreenCurve {keys [<at 0 2.5 RCIM_Cubic>,<at 1 2.5 RCIM_Cubic>]}
         RedCurve {keys [<at 0 2.3958 RCIM_Cubic>,<at 1 2.5 RCIM_Cubic>]}
        }}))
     "Float (Alpha)"=&Multiply_Float(
      A=&Lerp_Float(
       A=$System.SwingGeneralFade1
       B=$System.SwingGeneralFade2
       Alpha=&MakeFloatFromInt(Int=&SpawnGroup())
      )
      B=&FloatFromCurve(
       FloatCurve=DataInterface() {propertyValues {Curve {keys [<at 0 0>,<at 0.07 1>]}}}
       CurveIndex=$Emitter.Age
       Scale Curve=&Multiply_Float(A=3, B=$System.ToDScalar)
      )
     )
    )
    Create Color Attribute=true
    Write to Module Namespace=true
   ) {module=FN_Sprite_Attributes002}
   &DynamicMaterialParameters(
    Index 0 Param 1=1
    Index 1 Param 1=3
    Index 1 Param 2=3
    Index 2 Param 1=100
    Index 2 Param 2=7
    Index 2 Param 4=0.6
    Module.Index0Float0Write=true
    Module.Index0Float1Write=true
    Module.Index0Float2Write=true
    Module.Index0Float3Write=true
    Module.Index1Float0Write=true
    Module.Index1Float1Write=true
    Module.Index1Float2Write=true
    Module.Index1Float3Write=true
    Module.Index2Float0Write=true
    Module.Index2Float1Write=true
    Module.Index2Float2Write=false
    Module.Index2Float3Write=false
    Write Parameter Index 1=::NewEnumerator1
    Write Parameter Index 2=::NewEnumerator1
   )
   &ScaleColor(
    Scale Alpha=&Multiply_Float(
     A=&Lerp_Float(
      A=$System.SwingGeneralFade1
      B=$System.SwingGeneralFade2
      Alpha=&MakeFloatFromInt(Int=&SpawnGroup())
     )
     B=&Multiply_Float(A=3, B=$System.ToDScalar)
    )
    Scale RGB=&MakeVectorFromLinearColorRGB(Color=&ColorFromCurve(ColorCurve=DataInterface() {propertyValues {
        AlphaCurve {keys []}
        BlueCurve {keys [<at 0.1224 0.4 RCIM_Cubic>,<at 1 1 RCIM_Cubic>]}
        GreenCurve {keys [<at 0.1224 1 RCIM_Cubic>,<at 1 1 RCIM_Cubic>]}
        RedCurve {keys [<at 0.1224 0.45 RCIM_Cubic>,<at 1 1 RCIM_Cubic>]}
       }}))
    ScaleA=true
    ScaleRGB=true
   )
   /SetVariables(Particles.DynamicParametersAlt:Vector4f=&MakeVector4(X=35, Y=7, Z=0, W=0)) {isSetParameters=true}
   /SetVariables(Particles.RibbonWidthAlt2:NiagaraFloat=&Multiply_Float(A=$Particles.RibbonWidth, B=0.525)) {isSetParameters=true}
   &FN_Sprite_Attributes(
    Color=&MultiplyLinearColor(
     Linear Color=<rgba 0.069 0.6 0.1575 1>
     Scale Factor=&MakeVector4FromVectorAndFloat(
      "Vector (RGB)"=<1 1 1>
      "Float (Alpha)"=&Multiply_Float(
       A=&Lerp_Float(
        A=$System.SwingGeneralFade1
        B=$System.SwingGeneralFade2
        Alpha=&MakeFloatFromInt(Int=&SpawnGroup())
       )
       B=$System.ToDScalar
      )
     )
    )
    Create Color Attribute=true
    Write to Module Namespace=true
   )
   &self:LightningJitterPosition(
    Jitter Amount:NiagaraFloat=&FloatFromCurve(FloatCurve=DataInterface() {propertyValues {Curve {keys [<at 0 0>,<at 0.05 1>,<at 0.8 1>,<at 1 0>]}}}, Scale Curve=8)
    Jitter Offset:Vector3f=&Multiply_Vector(A=&RandomVector(Vector Scale=1), B=$Particles.RibbonFacing)
   ) {module=LightningJitterPosition}
   /SetVariables(Particles.RibbonWidthAlt4:NiagaraFloat=&Multiply_Float(
     A=$Particles.Initial.RibbonWidth
     B=&FloatFromCurve(FloatCurve=DataInterface() {propertyValues {Curve {keys [<at 0 1 RCIM_Cubic>,<at 0.4 3 RCIM_Cubic>]}}}, Scale Curve=0.3)
    )) {isSetParameters=true}
   /SetVariables(
    Particles.DynamicParametersAlt8:Vector4f=&MakeVector4(
     X=0
     Y=0
     Z=&Add_Float(
      A=&FloatFromCurve(FloatCurve=DataInterface() {propertyValues {Curve {keys [<at 0 0>,<at 1 1>]}}}, CurveIndex=$Emitter.Age, Scale Curve=-1)
      B=$Emitter.RandomUOffset
     )
     W=0
    )
    Particles.DynamicParametersAlt9:Vector4f=&MakeVector4(X=30, Y=9.5, Z=0, W=0)
   ) {isSetParameters=true}
   &FN_Sprite_Attributes(
    Color=&MakeLinearColorFromVectorAndFloat(
     "Vector (RGB)"=&Multiply_VectorByFloat(
      Vector=&MakeVectorFromLinearColorRGB(Color=&ColorFromCurve(ColorCurve=DataInterface() {propertyValues {
          AlphaCurve {keys []}
          BlueCurve {keys [<at 0 3.7905 RCIM_Cubic>,<at 0.5488 2.527>,<at 0.9992 0 RCIM_Cubic>]}
          GreenCurve {keys [<at 0 15 RCIM_Cubic>,<at 0.5488 10>,<at 0.9992 5 RCIM_Cubic>]}
          RedCurve {keys [<at 0 7.1321 RCIM_Cubic>,<at 0.5488 4.7547>,<at 0.9992 1.4905 RCIM_Cubic>]}
         }}))
      Float=&Multiply_Float(
       A=&Lerp_Float(
        A=$System.SwingGeneralFade1
        B=$System.SwingGeneralFade2
        Alpha=&MakeFloatFromInt(Int=&SpawnGroup())
       )
       B=5
      )
     )
     "Float (Alpha)"=&Multiply_Float(
      A=&Lerp_Float(
       A=$System.SwingGeneralFade1
       B=$System.SwingGeneralFade2
       Alpha=&MakeFloatFromInt(Int=&SpawnGroup())
      )
      B=$System.ToDScalar
     )
    )
    Create Color Attribute=true
    Write to Module Namespace=true
   ) {module=FN_Sprite_Attributes003}
  ]
  rendererClasses [&NiagaraRibbonRendererProperties]
  renderers [
   {rendererClass=&NiagaraRibbonRendererProperties}
   {rendererClass=&NiagaraRibbonRendererProperties, rendererIndex=1}
   {rendererClass=&NiagaraRibbonRendererProperties, rendererIndex=2}
   {rendererClass=&NiagaraRibbonRendererProperties, rendererIndex=3}
   {rendererClass=&NiagaraRibbonRendererProperties, rendererIndex=4}
  ]
 }
 TipTrail {
  emitterSpawnScript [&FN_RibbonIDOffsetUtility()]
  emitterUpdateScript [
   &EmitterState()
   &FN_SpawnAnimTrail(AnimTrailWidthMode=::FromSecond, Number of Anim Trails=2)
  ]
  particleSpawnScript [
   &FN_InitializeAnimTrailBoneIndecies()
   &FN_InitializeAnimTrailBoneIndecies(AnimTrail Selection=1, First Bone or Socket Index=2, Second Bone or Socket Index=3) {module=FN_InitializeAnimTrailBoneIndecies001}
   &FN_InitializeAnimTrail(AnimTrailWidth=0.3, Color=<rgba 0.0409 0.5009 0 1>)
  ]
  particleUpdateScript [
   &FN_UpdateAnimTrail(Lock Particles=true) {module=FN_UpdateAnimTrail001}
   &JitterPosition(
    Jitter Amount:NiagaraFloat=&FloatFromCurve(FloatCurve=DataInterface() {propertyValues {Curve {keys [<at 0 0>,<at 0.05 1>,<at 0.8 1>,<at 1 0>]}}}, Scale Curve=5)
    Jitter Offset:Vector3f=&Multiply_Vector(A=$System.RandomJitterVector, B=$Particles.RibbonFacing)
   ) {enabled=false}
   &DynamicMaterialParameters(
    Index 0 Param 1=1
    Index 1 Param 1=3
    Index 1 Param 2=3
    Index 1 Param 4=1
    Index 2 Param 1=100
    Index 2 Param 2=30
    Index 2 Param 4=0.6
    Module.Index0Float0Write=true
    Module.Index0Float1Write=true
    Module.Index0Float2Write=true
    Module.Index0Float3Write=true
    Module.Index1Float0Write=true
    Module.Index1Float1Write=true
    Module.Index1Float2Write=true
    Module.Index1Float3Write=true
    Module.Index2Float0Write=true
    Module.Index2Float1Write=true
    Module.Index2Float2Write=false
    Module.Index2Float3Write=false
    Write Parameter Index 1=::NewEnumerator1
    Write Parameter Index 2=::NewEnumerator1
   )
   &ScaleColor(
    Scale Alpha=&Lerp_Float(
     A=$System.SwingGeneralFade1
     B=$System.SwingGeneralFade2
     Alpha=&MakeFloatFromInt(Int=&SpawnGroup())
    )
    ScaleA=true
    ScaleRGB=true
   ) {module=ScaleColor001}
  ]
  rendererClasses [&NiagaraRibbonRendererProperties]
  renderers [
   {rendererClass=&NiagaraRibbonRendererProperties}
   {rendererClass=&NiagaraRibbonRendererProperties, rendererIndex=1}
  ]
 }
}
summary {
 systemName=NS_Emote_AlderOptic_Green_Trail_1
 emitters [
  {emitterName=MainTrail_Packed, rendererClasses [&NiagaraRibbonRendererProperties]}
  {emitterName=TipTrail, rendererClasses [&NiagaraRibbonRendererProperties]}
 ]
 userVariables [
  User.RibbonIDOffset {type=NiagaraInt32}
  User.SkeletalMesh {
   type=NiagaraDataInterfaceSkeletalMesh
   defaultValue {
    struct=NiagaraExt_VariableValue_DataInterface
    value {dataInterface=&self:NiagaraDataInterfaceSkeletalMesh_0}
   }
  }
  User.SwingSectionIndex {type=NiagaraInt32}
 ]
}
